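# Animation Blueprint for the HighMotion pickaxe cosmetic's skeleton, posing the weapon and its whip bones.
# It offsets the input pose's root and caches it.
# bIsInFrontEnd selects between an additive whip_skin_04 flip when true and the idle pose layered from b_AR_Root when false.
# DefaultSlot montages then layer from whip_skin_01, before a constraint node and fixed rotation are applied to Melee_TwoHanded_Handle.
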AnimBlueprint /BRCosmetics/Weapons/FORT_Melee/Pickaxe_HighMotion/Meshes/HighMotion_Axe_AnimBP
target_skeleton=&/HighMotion_Axe_Skeleton
graphs [
 AnimGraph {nodes [
   n1 = "Transform (Modify) Bone - Bone: whip_skin_04"(
    ComponentPose=Local To Component(LocalPose=Use cached pose 'SavedPose21808'())
    Alpha=1
    Translation="0,0,0"
    Rotation="180,0,0"
    Scale="1,1,1"
   ) {
    rotationMode=BMM_Additive
    rotationSpace=BCS_BoneSpace
    translationSpace=BCS_BoneSpace
    alphaBoolBlend {blendOption=Linear}
    boneToModify {boneName=whip_skin_04}
    rotation {pitch=180}
   }
   n2 = Sequence Player 'HighMotion_Idle_Pose_WEP'() {bLoopAnimation=false, sequence=/BRCosmetics/Animation/Game/MainPlayer/Combat/Melee/OneHanded/HighMotion/HighMotion_Idle_Pose_WEP}
   n3 = Use cached pose 'SavedPose21808'()
   n4 = Layered blend per bone(
    BasePose=Blend Poses by bool(
     BlendPose_0=Component To Local(ComponentPose=@n1)
     BlendPose_1=Layered blend per bone(BasePose=@n3, BlendPoses_0=@n2) {layerSetup [{branchFilters [{boneName=b_AR_Root}]}]}
     bActiveValue=Get bIsInFrontEnd()
    ) {blendType=Linear}
    BlendPoses_0=Slot 'DefaultSlot'()
   ) {layerSetup [{branchFilters [{boneName=whip_skin_01}]}]}
   n5 = "Transform (Modify) Bone - Bone: Melee_TwoHanded_Handle"(
    ComponentPose="Constraint - Melee_TwoHanded_Handle ()"!(ComponentPose=Local To Component(LocalPose=@n4), Alpha=1) -> (Pose) {alphaBoolBlend {blendOption=Linear}, boneToModify {boneName=Melee_TwoHanded_Handle}}
    Alpha=1
    Translation="0,0,0"
    Rotation="0,-90,0"
    Scale="1,1,1"
   ) {rotationMode=BMM_Replace, alphaBoolBlend {blendOption=Linear}, boneToModify {boneName=Melee_TwoHanded_Handle}, rotation {yaw=-90}}
   Output Pose(Result=Component To Local(ComponentPose=@n5)) {layerGroup=None}
   AnimGraphNode_SubInput_0 = Input Pose()
   n6 = "Transform (Modify) Bone - Bone: root"(ComponentPose=Local To Component(LocalPose=@AnimGraphNode_SubInput_0), Translation="2.291,-1.763,15.724", Rotation="0,0,0") {
    translationMode=BMM_Replace
    translationSpace=BCS_BoneSpace
    alphaBoolBlend {blendOption=Linear}
    boneToModify {boneName=root}
    translation=<2.291 -1.763 15.724>
   }
   Save cached pose 'SavedPose21808'(Pose=Component To Local(ComponentPose=@n6)) {cacheName=SavedPose21808}
  ]}
 EventGraph {links [], nodes []}
]
variables []
